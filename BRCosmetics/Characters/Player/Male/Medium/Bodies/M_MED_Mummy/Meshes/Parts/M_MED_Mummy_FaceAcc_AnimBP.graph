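# Animation Blueprint for the male medium Mummy character's face accessory mesh, adding secondary physics motion to its dyn_hair bones.
# It copies the pose from the attached parent mesh, then runs serial Anim Dynamics chain and single-bone nodes across front, side, back and upper bones.
# Each uses root-relative simulation, angular springs, per-bone limits, overridden gravity and damped, clamped component motion.
# The event graph has no logic.
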AnimBlueprint /BRCosmetics/Characters/Player/Male/Medium/Bodies/M_MED_Mummy/Meshes/Parts/M_MED_Mummy_FaceAcc_AnimBP
target_skeleton=/Game/Characters/Player/Male/Medium/Bodies/M_MED_Mummy/Meshes/Parts/M_MED_Mummy_FaceAcc_Skeleton
graphs [
 AnimGraph {nodes [
   n1 = "Anim Dynamics - Chain: dyn_hair_front_r_01 -> dyn_hair_front_r_02"(ComponentPose=Local To Component(LocalPose=Copy Pose From Mesh() {bUseAttachedParent=true}), Alpha=1, ExternalForce="0,0,0") {
    angularBiasOverride=0.9
    angularSpringConstant=1000
    bAngularSpring=true
    bChain=true
    bOverrideAngularBias=true
    bUseGravityOverride=true
    simulationSpace=RootRelative
    alphaBoolBlend {blendOption=Linear}
    boundBone {boneName=dyn_hair_front_r_01}
    chainEnd {boneName=dyn_hair_front_r_02}
    componentAppliedLinearAccClamp=<400 400 400>
    componentLinearAccScale=<0.6 0.6 0.6>
    componentLinearVelScale=<0.6 0.6 0.6>
    gravityOverride {z=-500}
    physicsBodyDefinitions [
     {
      boundBone {boneName=dyn_hair_front_r_01}
      boxExtents=<8 16 16>
      constraintSetup {angularLimitsMax=<3 2 10>, angularLimitsMin=<-3 -15 0>, angularTarget=<0 0 0>}
      localJointOffset=<1 0 0>
     }
     {
      boundBone {boneName=dyn_hair_front_r_02}
      boxExtents=<8 16 16>
      constraintSetup {angularLimitsMax=<3 2 10>, angularLimitsMin=<-3 -15 0>, angularTarget=<0 0 0>}
      localJointOffset=<1 0 0>
     }
    ]
    relativeSpaceBone {boneName=head}
   }
   n2 = "Anim Dynamics - Chain: dyn_hair_side_r_01 -> dyn_hair_side_r_02"(ComponentPose=@n1, Alpha=1, ExternalForce="0,0,0") {
    angularBiasOverride=0.9
    angularSpringConstant=1000
    bAngularSpring=true
    bChain=true
    bOverrideAngularBias=true
    bUseGravityOverride=true
    simulationSpace=RootRelative
    alphaBoolBlend {blendOption=Linear}
    boundBone {boneName=dyn_hair_side_r_01}
    chainEnd {boneName=dyn_hair_side_r_02}
    componentAppliedLinearAccClamp=<400 400 400>
    componentLinearAccScale=<0.6 0.6 0.6>
    componentLinearVelScale=<0.6 0.6 0.6>
    gravityOverride {z=-500}
    physicsBodyDefinitions [
     {
      boundBone {boneName=dyn_hair_side_r_01}
      boxExtents=<8 16 16>
      constraintSetup {angularLimitsMax=<0 3.5 5>, angularLimitsMin=<0 -3.5 -5>, angularTarget=<0 0 0>}
      localJointOffset=<1 0 0>
     }
     {
      boundBone {boneName=dyn_hair_side_r_02}
      boxExtents=<8 16 16>
      constraintSetup {angularLimitsMax=<0 3.5 5>, angularLimitsMin=<0 -3.5 -5>, angularTarget=<0 0 0>}
      localJointOffset=<1 0 0>
     }
    ]
    relativeSpaceBone {boneName=head}
   }
   n3 = Anim Dynamics - Bone: dyn_hair_side_r_03(ComponentPose=@n2, Alpha=1, ExternalForce="0,0,0") {
    angularBiasOverride=0.9
    angularSpringConstant=1000
    bAngularSpring=true
    bOverrideAngularBias=true
    bUseGravityOverride=true
    simulationSpace=RootRelative
    alphaBoolBlend {blendOption=Linear}
    boundBone {boneName=dyn_hair_side_r_03}
    componentAppliedLinearAccClamp=<400 400 400>
    componentLinearAccScale=<0.6 0.6 0.6>
    componentLinearVelScale=<0.6 0.6 0.6>
    gravityOverride {z=-500}
    physicsBodyDefinitions [{
      boundBone {boneName=dyn_hair_side_r_03}
      boxExtents=<8 16 16>
      constraintSetup {angularLimitsMax=<3 2 10>, angularLimitsMin=<-3 -15 0>, angularTarget=<0 0 0>}
      localJointOffset=<1 0 0>
     }]
    relativeSpaceBone {boneName=head}
   }
   n4 = "Anim Dynamics - Chain: dyn_hair_side_r_04 -> dyn_hair_side_r_05"(ComponentPose=@n3, Alpha=1, ExternalForce="0,0,0") {
    angularBiasOverride=0.9
    angularSpringConstant=1000
    bAngularSpring=true
    bChain=true
    bOverrideAngularBias=true
    bUseGravityOverride=true
    simulationSpace=RootRelative
    alphaBoolBlend {blendOption=Linear}
    boundBone {boneName=dyn_hair_side_r_04}
    chainEnd {boneName=dyn_hair_side_r_05}
    componentAppliedLinearAccClamp=<400 400 400>
    componentLinearAccScale=<0.6 0.6 0.6>
    componentLinearVelScale=<0.6 0.6 0.6>
    gravityOverride {z=-500}
    physicsBodyDefinitions [
     {
      boundBone {boneName=dyn_hair_side_r_04}
      boxExtents=<8 16 16>
      constraintSetup {angularLimitsMax=<10 0 15>, angularLimitsMin=<-10 -25 5>, angularTarget=<0 0 0>}
      localJointOffset=<1 0 0>
     }
     {
      boundBone {boneName=dyn_hair_side_r_05}
      boxExtents=<8 16 16>
      constraintSetup {angularLimitsMax=<10 10 15>, angularLimitsMin=<-10 -10 5>, angularTarget=<0 0 0>}
      localJointOffset=<1 0 0>
     }
    ]
    relativeSpaceBone {boneName=head}
   }
   n5 = "Anim Dynamics - Chain: dyn_hair_back_r_01 -> dyn_hair_back_r_02"(ComponentPose=@n4, Alpha=1, ExternalForce="0,0,0") {
    angularBiasOverride=0.9
    angularSpringConstant=1000
    bAngularSpring=true
    bChain=true
    bOverrideAngularBias=true
    bUseGravityOverride=true
    simulationSpace=RootRelative
    alphaBoolBlend {blendOption=Linear}
    boundBone {boneName=dyn_hair_back_r_01}
    chainEnd {boneName=dyn_hair_back_r_02}
    componentAppliedLinearAccClamp=<400 400 400>
    componentLinearAccScale=<0.6 0.6 0.6>
    componentLinearVelScale=<0.6 0.6 0.6>
    gravityOverride {z=-500}
    physicsBodyDefinitions [
     {
      boundBone {boneName=dyn_hair_back_r_01}
      boxExtents=<8 16 16>
      constraintSetup {angularLimitsMax=<10 10 10>, angularLimitsMin=<-10 -10 -10>, angularTarget=<0 0 0>}
      localJointOffset=<1 0 0>
     }
     {
      boundBone {boneName=dyn_hair_back_r_02}
      boxExtents=<8 16 16>
      constraintSetup {angularLimitsMax=<10 10 10>, angularLimitsMin=<-10 -10 -10>, angularTarget=<0 0 0>}
      localJointOffset=<1 0 0>
     }
    ]
    relativeSpaceBone {boneName=head}
   }
   n6 = Anim Dynamics - Bone: dyn_hair_upper_01(ComponentPose=@n5, Alpha=1, ExternalForce="0,0,0") {
    angularBiasOverride=0.9
    angularSpringConstant=1000
    bAngularSpring=true
    bOverrideAngularBias=true
    bUseGravityOverride=true
    simulationSpace=RootRelative
    alphaBoolBlend {blendOption=Linear}
    boundBone {boneName=dyn_hair_upper_01}
    componentAppliedLinearAccClamp=<400 400 400>
    componentLinearAccScale=<0.6 0.6 0.6>
    componentLinearVelScale=<0.6 0.6 0.6>
    gravityOverride {z=-500}
    physicsBodyDefinitions [{
      boundBone {boneName=dyn_hair_upper_01}
      boxExtents=<8 16 16>
      constraintSetup {angularLimitsMax=<10 10 10>, angularLimitsMin=<-10 -10 -10>, angularTarget=<0 0 0>}
      localJointOffset=<1 0 0>
     }]
    relativeSpaceBone {boneName=head}
   }
   n7 = "Anim Dynamics - Chain: dyn_hair_back_l_01 -> dyn_hair_back_l_02"(ComponentPose=@n6, Alpha=1, ExternalForce="0,0,0") {
    angularBiasOverride=0.9
    angularSpringConstant=1000
    bAngularSpring=true
    bChain=true
    bOverrideAngularBias=true
    bUseGravityOverride=true
    simulationSpace=RootRelative
    alphaBoolBlend {blendOption=Linear}
    boundBone {boneName=dyn_hair_back_l_01}
    chainEnd {boneName=dyn_hair_back_l_02}
    componentAppliedLinearAccClamp=<400 400 400>
    componentLinearAccScale=<0.6 0.6 0.6>
    componentLinearVelScale=<0.6 0.6 0.6>
    gravityOverride {z=-500}
    physicsBodyDefinitions [
     {
      boundBone {boneName=dyn_hair_back_l_01}
      boxExtents=<8 16 16>
      constraintSetup {angularLimitsMax=<10 5 -8>, angularLimitsMin=<-10 -15 -20>, angularTarget=<0 0 0>}
      localJointOffset=<1 0 0>
     }
     {
      boundBone {boneName=dyn_hair_back_l_02}
      boxExtents=<8 16 16>
      constraintSetup {angularLimitsMax=<10 5 -10>, angularLimitsMin=<-10 -15 -23>, angularTarget=<0 0 0>}
      localJointOffset=<1 0 0>
     }
    ]
    relativeSpaceBone {boneName=head}
   }
   n8 = "Anim Dynamics - Chain: dyn_hair_upper_02 -> dyn_hair_upper_03"(ComponentPose=@n7, Alpha=1, ExternalForce="0,0,0") {
    angularBiasOverride=0.9
    angularSpringConstant=1000
    bAngularSpring=true
    bChain=true
    bOverrideAngularBias=true
    bUseGravityOverride=true
    simulationSpace=RootRelative
    alphaBoolBlend {blendOption=Linear}
    boundBone {boneName=dyn_hair_upper_02}
    chainEnd {boneName=dyn_hair_upper_03}
    componentAppliedLinearAccClamp=<400 400 400>
    componentLinearAccScale=<0.6 0.6 0.6>
    componentLinearVelScale=<0.6 0.6 0.6>
    gravityOverride {z=-500}
    physicsBodyDefinitions [
     {
      boundBone {boneName=dyn_hair_upper_02}
      boxExtents=<8 16 16>
      constraintSetup {angularLimitsMax=<10 10 -2>, angularLimitsMin=<-10 -10 -10>, angularTarget=<0 0 0>}
      localJointOffset=<1 0 0>
     }
     {
      boundBone {boneName=dyn_hair_upper_03}
      boxExtents=<8 16 16>
      constraintSetup {angularLimitsMax=<10 10 -2>, angularLimitsMin=<-10 -10 -10>, angularTarget=<0 0 0>}
      localJointOffset=<1 0 0>
     }
    ]
    relativeSpaceBone {boneName=head}
   }
   n9 = "Anim Dynamics - Chain: dyn_hair_upper_04 -> dyn_hair_upper_05"(ComponentPose=@n8, Alpha=1, ExternalForce="0,0,0") {
    angularBiasOverride=0.9
    angularSpringConstant=1000
    bAngularSpring=true
    bChain=true
    bOverrideAngularBias=true
    bUseGravityOverride=true
    simulationSpace=RootRelative
    alphaBoolBlend {blendOption=Linear}
    boundBone {boneName=dyn_hair_upper_04}
    chainEnd {boneName=dyn_hair_upper_05}
    componentAppliedLinearAccClamp=<400 400 400>
    componentLinearAccScale=<0.6 0.6 0.6>
    componentLinearVelScale=<0.6 0.6 0.6>
    gravityOverride {z=-500}
    physicsBodyDefinitions [
     {
      boundBone {boneName=dyn_hair_upper_04}
      boxExtents=<8 16 16>
      constraintSetup {angularLimitsMax=<10 10 2>, angularLimitsMin=<-10 -10 -2>, angularTarget=<0 0 0>}
      localJointOffset=<1 0 0>
     }
     {
      boundBone {boneName=dyn_hair_upper_05}
      boxExtents=<8 16 16>
      constraintSetup {angularLimitsMax=<10 10 10>, angularLimitsMin=<-10 -10 -10>, angularTarget=<0 0 0>}
      localJointOffset=<1 0 0>
     }
    ]
    relativeSpaceBone {boneName=head}
   }
   n10 = Anim Dynamics - Bone: dyn_hair_side_l_01(ComponentPose=@n9, Alpha=1, ExternalForce="0,0,0") {
    angularBiasOverride=0.9
    angularSpringConstant=1000
    bAngularSpring=true
    bOverrideAngularBias=true
    bUseGravityOverride=true
    simulationSpace=RootRelative
    alphaBoolBlend {blendOption=Linear}
    boundBone {boneName=dyn_hair_side_l_01}
    componentAppliedLinearAccClamp=<400 400 400>
    componentLinearAccScale=<0.6 0.6 0.6>
    componentLinearVelScale=<0.6 0.6 0.6>
    gravityOverride {z=-500}
    physicsBodyDefinitions [{
      boundBone {boneName=dyn_hair_side_l_01}
      boxExtents=<8 16 16>
      constraintSetup {angularLimitsMax=<10 5 10>, angularLimitsMin=<-10 -10 -10>, angularTarget=<0 0 0>}
      localJointOffset=<1 0 0>
     }]
    relativeSpaceBone {boneName=head}
   }
   n11 = Anim Dynamics - Bone: dyn_hair_side_l_02(ComponentPose=@n10, Alpha=1, ExternalForce="0,0,0") {
    angularBiasOverride=0.9
    angularSpringConstant=1000
    bAngularSpring=true
    bOverrideAngularBias=true
    bUseGravityOverride=true
    simulationSpace=RootRelative
    alphaBoolBlend {blendOption=Linear}
    boundBone {boneName=dyn_hair_side_l_02}
    componentAppliedLinearAccClamp=<400 400 400>
    componentLinearAccScale=<0.6 0.6 0.6>
    componentLinearVelScale=<0.6 0.6 0.6>
    gravityOverride {z=-500}
    physicsBodyDefinitions [{
      boundBone {boneName=dyn_hair_side_l_02}
      boxExtents=<8 16 16>
      constraintSetup {angularLimitsMax=<10 0 -5>, angularLimitsMin=<-10 -30 -25>, angularTarget=<0 0 0>}
      localJointOffset=<1 0 0>
     }]
    relativeSpaceBone {boneName=head}
   }
   n12 = Anim Dynamics - Bone: dyn_hair_side_l_03(ComponentPose=@n11, Alpha=1, ExternalForce="0,0,0") {
    angularBiasOverride=0.9
    angularSpringConstant=1000
    bAngularSpring=true
    bOverrideAngularBias=true
    bUseGravityOverride=true
    simulationSpace=RootRelative
    alphaBoolBlend {blendOption=Linear}
    boundBone {boneName=dyn_hair_side_l_03}
    componentAppliedLinearAccClamp=<400 400 400>
    componentLinearAccScale=<0.6 0.6 0.6>
    componentLinearVelScale=<0.6 0.6 0.6>
    gravityOverride {z=-500}
    physicsBodyDefinitions [{
      boundBone {boneName=dyn_hair_side_l_03}
      boxExtents=<8 16 16>
      constraintSetup {angularLimitsMax=<10 10 0>, angularLimitsMin=<-10 -10 -12>, angularTarget=<0 0 0>}
      localJointOffset=<1 0 0>
     }]
    relativeSpaceBone {boneName=head}
   }
   n13 = "Anim Dynamics - Chain: dyn_hair_front_upper_l_01 -> dyn_hair_front_upper_l_02"(ComponentPose=@n12, Alpha=1, ExternalForce="0,0,0") {
    angularBiasOverride=0.999
    angularSpringConstant=1000
    bAngularSpring=true
    bChain=true
    bOverrideAngularBias=true
    bUseGravityOverride=true
    simulationSpace=RootRelative
    alphaBoolBlend {blendOption=Linear}
    boundBone {boneName=dyn_hair_front_upper_l_01}
    chainEnd {boneName=dyn_hair_front_upper_l_02}
    componentAppliedLinearAccClamp=<400 400 400>
    componentLinearAccScale=<0.6 0.6 0.6>
    componentLinearVelScale=<0.6 0.6 0.6>
    gravityOverride {z=-500}
    physicsBodyDefinitions [
     {
      boundBone {boneName=dyn_hair_front_upper_l_01}
      boxExtents=<8 16 16>
      constraintSetup {angularLimitsMax=<10 0 5>, angularLimitsMin=<-10 -15 -10>, angularTarget=<0 0 0>}
      localJointOffset=<1 0 0>
     }
     {
      boundBone {boneName=dyn_hair_front_upper_l_02}
      boxExtents=<8 16 16>
      constraintSetup {angularLimitsMax=<10 0 5>, angularLimitsMin=<-10 -15 -10>, angularTarget=<0 0 0>}
      localJointOffset=<1 0 0>
     }
    ]
    relativeSpaceBone {boneName=head}
   }
   n14 = "Anim Dynamics - Chain: dyn_hair_front_l_01 -> dyn_hair_front_l_02"(ComponentPose=@n13, Alpha=1, ExternalForce="0,0,0") {
    angularBiasOverride=0.999
    angularSpringConstant=1000
    bAngularSpring=true
    bChain=true
    bOverrideAngularBias=true
    bUseGravityOverride=true
    simulationSpace=RootRelative
    alphaBoolBlend {blendOption=Linear}
    boundBone {boneName=dyn_hair_front_l_01}
    chainEnd {boneName=dyn_hair_front_l_02}
    componentAppliedLinearAccClamp=<400 400 400>
    componentLinearAccScale=<0.6 0.6 0.6>
    componentLinearVelScale=<0.6 0.6 0.6>
    gravityOverride {z=-400}
    physicsBodyDefinitions [
     {
      boundBone {boneName=dyn_hair_front_l_01}
      boxExtents=<4 8 8>
      constraintSetup {angularLimitsMax=<5 10 25>, angularLimitsMin=<-5 -30 -5>, angularTarget=<0 0 0>}
      localJointOffset=<1 0 0>
     }
     {
      boundBone {boneName=dyn_hair_front_l_02}
      boxExtents=<4 8 8>
      constraintSetup {angularLimitsMax=<5 10 25>, angularLimitsMin=<-5 -30 -5>, angularTarget=<0 0 0>}
      localJointOffset=<1 0 0>
     }
    ]
    relativeSpaceBone {boneName=head}
   }
   n15 = Anim Dynamics - Bone: dyn_hair_front_l_03(ComponentPose=@n14, Alpha=1, ExternalForce="0,0,0") {
    angularBiasOverride=0.9
    angularSpringConstant=1000
    bAngularSpring=true
    bOverrideAngularBias=true
    bUseGravityOverride=true
    simulationSpace=RootRelative
    alphaBoolBlend {blendOption=Linear}
    boundBone {boneName=dyn_hair_front_l_03}
    componentAppliedLinearAccClamp=<400 400 400>
    componentLinearAccScale=<0.6 0.6 0.6>
    componentLinearVelScale=<0.6 0.6 0.6>
    gravityOverride {z=-400}
    physicsBodyDefinitions [{
      boundBone {boneName=dyn_hair_front_l_03}
      boxExtents=<8 16 16>
      constraintSetup {angularLimitsMax=<10 -5 -5>, angularLimitsMin=<-10 -20 -20>, angularTarget=<0 0 0>}
      localJointOffset=<10 0 0>
     }]
    relativeSpaceBone {boneName=head}
   }
   n16 = Anim Dynamics - Bone: dyn_hair_front_l_04(ComponentPose=@n15, Alpha=1, ExternalForce="0,0,0") {
    angularBiasOverride=0.9
    angularSpringConstant=1000
    bAngularSpring=true
    bOverrideAngularBias=true
    bUseGravityOverride=true
    simulationSpace=RootRelative
    alphaBoolBlend {blendOption=Linear}
    boundBone {boneName=dyn_hair_front_l_04}
    componentAppliedLinearAccClamp=<400 400 400>
    componentLinearAccScale=<0.6 0.6 0.6>
    componentLinearVelScale=<0.6 0.6 0.6>
    gravityOverride {z=-400}
    physicsBodyDefinitions [{
      boundBone {boneName=dyn_hair_front_l_04}
      boxExtents=<10 16 16>
      constraintSetup {angularLimitsMax=<10 60 5>, angularLimitsMin=<-10 20 -10>, angularTarget=<0 0 0>}
      localJointOffset=<10 0 5>
     }]
    relativeSpaceBone {boneName=head}
   }
   Output Pose(Result=Component To Local(ComponentPose=@n16)) {layerGroup=None}
  ]}
 EventGraph {nodes [Event BlueprintUpdateAnimation()]}
]
variables []
